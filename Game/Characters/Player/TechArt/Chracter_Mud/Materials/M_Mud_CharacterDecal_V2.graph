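# Translucent mud decal for characters, reading scene depth and normals behind a camera-facing mesh instead of using true decal projection.
# Opacity combines an actor-height gradient, panning drip noise, timestamped in/out fades, depth masks and custom-depth tests with a local-player blend.
# Desaturatable mud texture gets hand-built skylight, sun and specular shading for emissive.
# Custom primitive data or Niagara dynamic parameters drive it, with simpler Mobile and Forward fallbacks.
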Material /Game/Characters/Player/TechArt/Chracter_Mud/Materials/M_Mud_CharacterDecal_V2
graph {
 properties {
  allowTranslucentCustomDepthWrites=true
  bUsedWithNiagaraMeshParticles=true
  blendMode=BLEND_Translucent
  pixelDepthOffsetMode=PDOM_Legacy
  usedShadingModels=Default Lit
 }
 nodes [
  n1 = Constant3Vector() {constant=<rgba 0.5 0.5 1 1>}
  n2 = ShadingPathSwitch(Default=ComponentMask(None=SceneTexture() {sceneTextureId=PPI_WorldNormal}.Color) {b=true, g=true, r=true}, Forward=@n1, Mobile=@n1)
  n3 = MaterialFunctionCall(XY=TextureCoordinate()) {materialFunction=&Swizzle}
  n4 = VectorParameter() {parameterName=DetailNormalTilePanA, defaultValue=<rgba 1.115 2.9 0.01 0.505>}
  n5 = AppendVector(A=@n4.B, B=@n4.B)
  n6 = Panner(Coordinate=Multiply(A=@n3.YX, B=AppendVector(A=@n4.R, B=@n4.R)), Speed=@n5) {bFractionalPart=true}
  n7 = VectorParameter() {parameterName=DetailNormalTilePanB, defaultValue=<rgba -1.375 4 -0.024 0.288>}
  n8 = Panner(Coordinate=Multiply(A=@n3.YX, B=AppendVector(A=@n7.R, B=@n7.R)), Speed=AppendVector(A=@n7.B, B=@n7.B)) {bFractionalPart=true}
  n9 = Normalize(VectorInput=Add(
    A=TextureSample(UVs=@n6) {samplerType=SAMPLERTYPE_Normal, texture=/Game/Athena/Items/Weapons/Prototype/Papaya/T_PaintDrips_Normal}.RGB
    B=TextureSample(UVs=@n8) {samplerType=SAMPLERTYPE_Normal, texture=&T_PaintDrips_Normal}.RGB
   ))
  n10 = Constant()
  n11 = ShadingPathSwitch(Default=@n9, Forward=@n10, Mobile=@n10)
  Add(A=@n2, B=@n11)
  n12 = Constant() {r=3}
  Divide(A=Fmod(A=Time() {period=1}, B=@n12), B=@n12)
  MaterialFunctionCall(BaseNormal=@n2, AdditionalNormal=@n11) {materialFunction=&BlendAngleCorrectedNormals}
  MaterialFunctionCall!(Color, Luminosity) {materialFunction=/Engine/Functions/Engine_MaterialFunctions01/ImageAdjustment/Luminosity_And_Color}
  Multiply(A=@n9, B=VectorParameter() {parameterName=DetailNormalStrength, defaultValue=<rgba 0.8 0.8 0 0>}.RGB)
  n13 = CollectionParameter() {
   collection=&FortniteMaterialParameters
   parameterName=SunlightDirectionalVector
  }
  n14 = Normalize(VectorInput=@n2)
  n15 = Saturate(None=DotProduct(A=ComponentMask(None=@n13) {b=true, g=true, r=true}, B=@n14))
  n16 = ShadingPathSwitch(Default=Multiply(A=NamedRerouteUsage(), B=@n15), Mobile=@n15)
  n17 = Constant() {r=1}
  n18 = ShadingPathSwitch(Default=ComponentMask(None=SceneTexture() {sceneTextureId=PPI_AmbientOcclusion}.Color) {r=true}, Forward=@n17, Mobile=@n17)
  NamedRerouteDeclaration(None=Saturate(None=Multiply(A=@n16, B=@n18))) {name=SunDotMask, nodeColor=<rgba 1 0 0.1172 1>}
  n19 = ScalarParameter() {parameterName=LuminanceMin_Morning}
  n20 = ScalarParameter() {parameterName=LuminanceMin_Day}
  n21 = ScalarParameter() {parameterName=LuminanceMin_Evening}
  n22 = ScalarParameter() {parameterName=LuminanceMin_Night}
  n23 = ScalarParameter() {defaultValue=1, parameterName=LuminanceMax_Morning}
  n24 = ScalarParameter() {defaultValue=1, parameterName=LuminanceMax_Day}
  n25 = ScalarParameter() {defaultValue=1, parameterName=LuminanceMax_Evening}
  n26 = ScalarParameter() {defaultValue=1, parameterName=LuminanceMax_Night}
  n27 = Divide(A=AppendVector(A=SceneColor!(UVs).RGB, B=Constant()), B=SceneTexture() {sceneTextureId=PPI_BaseColor}.Color)
  NamedRerouteDeclaration(None=LinearInterpolate(
    A=MaterialFunctionCall(Color_Morning=@n19, Color_Day=@n20, Color_Evening=@n21, Color_Night=@n22) {materialFunction=&Color_Time_of_Day}.Result
    B=MaterialFunctionCall(Color_Morning=@n23, Color_Day=@n24, Color_Evening=@n25, Color_Night=@n26) {materialFunction=&Color_Time_of_Day}.Result
    Alpha=Saturate(None=@n27)
   )) {name=Luminance, nodeColor=<rgba 0 0.7656 1 1>}
  n28 = Multiply(A=TextureCoordinate(), B=ScalarParameter() {defaultValue=1, parameterName=Mud_Tiling})
  n29 = Multiply(
   A=MaterialFunctionCall(XY=@n5) {materialFunction=&Swizzle}.YX
   B=ScalarParameter() {defaultValue=0.2, parameterName=NoisePanMult}
  )
  n30 = Multiply(
   A=TextureSample(UVs=Panner(Coordinate=@n28, Speed=@n29) {speedX=1.5}) {texture=/Engine/MaterialTemplates/Textures/T_Noise01}.R
   B=ScalarParameter() {defaultValue=0.5, parameterName=OpacityNoiseStrength}
  )
  NamedRerouteDeclaration(None=@n30) {name=OpacityNoise, nodeColor=<rgba 0 0.5312 1 1>}
  n31 = Panner(Coordinate=@n28) {speedX=0.02}
  NamedRerouteDeclaration(None=MaterialFunctionCall(
    Normal=TextureSample(UVs=@n31) {samplerType=SAMPLERTYPE_Normal, texture=/Game/Environments/Asteria/Terrain/Textures/T_Asteria_MudGround_A_N}.RGB
    Flatness=ScalarParameter() {defaultValue=0.4, parameterName=Normal_Softness}
   ) {materialFunction=&FlattenNormal}.Result) {name=Normal, nodeColor=<rgba 1 0 0.1172 1>}
  n32 = TextureSample(UVs=@n31) {texture=/Game/Environments/Asteria/Terrain/Textures/T_Asteria_MudGround_A_D}
  n33 = Desaturation(None=@n32.RGB, Fraction=ScalarParameter() {defaultValue=1, parameterName=DesaturationAmt})
  n34 = VectorParameter() {parameterName=MudColor, defaultValue=<rgba 1 1 1 1>}
  NamedRerouteDeclaration(None=LinearInterpolate(A=@n32.RGB, B=Multiply(A=@n33, B=@n34.RGB), Alpha=ScalarParameter() {parameterName=UseDesaturation})) {name=Diffuse, nodeColor=<rgba 0 1 0.0625 1>}
  NamedRerouteUsage()
  NamedRerouteUsage()
  NamedRerouteUsage()
  n35 = DynamicParameter() {parameterIndex=1, defaultValue=<rgba 80 160 0 0>, paramNames [ActorBiasZ,ActorHeight,LocalPlayer,none]}
  n36 = StaticSwitchParameter(True=@n35.ActorBiasZ, False=ScalarParameter() {bUseCustomPrimitiveData=true, parameterName=ActorBiasZ, primitiveDataIndex=4}) {group=Niagara, parameterName=Use Niagara Dynamic Params}
  n37 = ScalarParameter() {defaultValue=20, parameterName=AdditionalBiasZ}
  n38 = Add(
   A=ActorPositionWS()
   B=MaterialFunctionCall(B=Add(A=@n36, B=@n37)) {materialFunction=&AppendMany}.RGB
  )
  n39 = Subtract(A=MaterialFunctionCall() {materialFunction=&WorldPositionBehindTranslucency}.Result, B=@n38)
  n40 = DynamicParameter() {parameterIndex=1, defaultValue=<rgba 80 160 0 0>, paramNames [ActorBiasZ,ActorHeight,LocalPlayer,none]}
  n41 = ScalarParameter() {bUseCustomPrimitiveData=true, parameterName=ActorHeight, primitiveDataIndex=5}
  n42 = Saturate(None=Divide(
    A=ComponentMask(None=@n39) {b=true}
    B=StaticSwitchParameter(True=@n40.ActorHeight, False=@n41) {group=Niagara, parameterName=Use Niagara Dynamic Params}
   ))
  n43 = DotProduct(
   A=Transform(None=@n2) {transformSourceType=TRANSFORMSOURCE_World, transformType=TRANSFORM_View}
   B=ScalarParameter() {defaultValue=0.06, parameterName=DotTest}
  )
  n44 = VectorParameter() {parameterName=MaskTilePanA, defaultValue=<rgba 0.3193 0.1191 -0.0204 0.124>}
  n45 = AppendVector(A=@n44.R, B=@n44.R)
  n46 = TextureSample(UVs=Panner(Coordinate=Multiply(A=@n45, B=@n3.YX), Speed=AppendVector(A=@n44.B, B=@n44.B)) {bFractionalPart=true}) {samplerType=SAMPLERTYPE_Normal, texture=&T_PaintDrips_Normal}
  n47 = VectorParameter() {parameterName=MaskTilePanB, defaultValue=<rgba 0.284 0.0995 0.035 0.0821>}
  n48 = AppendVector(A=@n47.R, B=@n47.R)
  n49 = TextureSample(UVs=Panner(Coordinate=Multiply(A=@n3.YX, B=@n48), Speed=AppendVector(A=@n47.B, B=@n47.B)) {bFractionalPart=true}) {samplerType=SAMPLERTYPE_Normal, texture=&T_PaintDrips_Normal}
  n50 = Multiply(
   A=ScalarParameter() {defaultValue=0.3, parameterName=MaskHighRemap}
   B=DotProduct(A=Add(A=@n46.RGB, B=@n49.RGB), B=Constant3Vector() {constant=<rgba 0.4 0.4 0 0>})
  )
  n51 = Constant()
  n52 = DynamicParameter() {defaultValue=<rgba 1 0.2 4 2>, paramNames [TimestampA, TimestampTransitionLengthIn, TimestampB, TimestampTransitionLengthOut]}
  n53 = ScalarParameter() {bUseCustomPrimitiveData=true, parameterName=TimestampA}
  n54 = ScalarParameter() {bUseCustomPrimitiveData=true, parameterName=TimestampTransitionLengthIn, primitiveDataIndex=1}
  n55 = ScalarParameter() {bUseCustomPrimitiveData=true, defaultValue=1, parameterName=TimestampB, primitiveDataIndex=2}
  n56 = ScalarParameter() {bUseCustomPrimitiveData=true, parameterName=TimestampTransitionLengthOut, primitiveDataIndex=3}
  n57 = MaterialFunctionCall(Float4=StaticSwitchParameter(
    True=@n52.RGBA
    False=MaterialFunctionCall(X=@n53, Y=@n54, Z=@n55, A=@n56) {materialFunction=&MakeFloat4}.Result
   ) {group=Niagara, parameterName=Use Niagara Dynamic Params}) {materialFunction=&BreakOutFloat4Components}
  n58 = MaterialFunctionCall(TimestampA=@n57.R, TransitionLength_In=@n57.R, TimestampB=@n57.R, TransitionLength_Out=@n57.R, Transition In Only?=StaticBool()) {materialFunction=&MF_Timestamp_InOut_01}
  n59 = SphereMask(
   A=Add(A=Add(A=@n42, B=@n43), B=ShadingPathSwitch(Default=@n50, Forward=@n51, Mobile=@n51))
   B=Constant()
   Radius=@n58.Linear
   Hardness=Constant() {r=0.85}
  )
  n60 = Subtract(A=MaterialFunctionCall() {materialFunction=&WorldPositionBehindTranslucency}.Result, B=@n38)
  n61 = Saturate(None=Divide(A=ComponentMask(None=@n60) {b=true}, B=ScalarParameter() {defaultValue=50, parameterName=ZMaskDivisor}))
  n62 = Multiply(A=@n58.Linear) {constB=2}
  n63 = TextureCoordinate()
  n64 = Constant() {r=0.5}
  n65 = Constant() {r=0.5}
  n66 = ScalarParameter() {defaultValue=50, parameterName=CameraOffsetBias}
  n67 = ScalarParameter() {defaultValue=30, parameterName=DistanceDivisor}
  n68 = ComponentMask(None=SceneTexture() {sceneTextureId=PPI_CustomDepth}.Color) {r=true}
  n69 = Saturate(None=Divide(A=@n68) {constB=10000})
  n70 = If(A=SceneDepth().R, B=@n68, "A > B"=Constant() {r=1}, "A < B"=Constant())
  n71 = DynamicParameter() {parameterIndex=1, defaultValue=<rgba 80 160 0 0>, paramNames [ActorBiasZ,ActorHeight,LocalPlayer,none]}
  n72 = ScalarParameter() {bUseCustomPrimitiveData=true, parameterName=LocalPlayer, primitiveDataIndex=10}
  n73 = ShadingPathSwitch(
   Default=Constant() {r=1}
   Deferred=LinearInterpolate(
    A=LinearInterpolate(A=@n69, Alpha=@n70)
    B=OneMinus(None=@n69)
    Alpha=StaticSwitchParameter(True=@n71.LocalPlayer, False=@n72) {group=Niagara, parameterName=Use Niagara Dynamic Params}
   ) {constA=1}
  )
  n74 = Multiply(
   A=Multiply(
    A=SphereMask(A=@n63, B=@n64, Radius=@n64, Hardness=@n65)
    B=MaterialFunctionCall!(Distance Bias=@n66, Distance Divisor=@n67, Scene Depth, Pixel Depth) {materialFunction=/Game/Effects/Fort_Effects/Materials/_Functions/MF_DistanceBehindTranslucency_01}.Inverted
   )
   B=@n73
  )
  n75 = NamedRerouteUsage()
  n76 = Multiply(A=Multiply(A=Multiply(A=@n59, B=@n61), B=Saturate(None=@n62)), B=Clamp(None=Subtract(A=@n74, B=@n75)))
  n77 = Multiply(A=@n76, B=ScalarParameter() {defaultValue=0.6, parameterName=MobileOpacity})
  n78 = LinearInterpolate(
   A=ScalarParameter() {defaultValue=25, parameterName=Depth Blend Max}
   B=ScalarParameter() {defaultValue=2, parameterName=Depth Blend Min}
   Alpha=@n73
  )
  Reroute(None=Multiply(A=ShadingPathSwitch(Default=@n76, Forward=@n77, Mobile=@n77), B=DepthFade(Opacity=@n73, FadeDistance=@n78)))
  n79 = NamedRerouteUsage()
  n80 = Multiply(
   A=ScalarParameter() {defaultValue=3, parameterName=BaseCXolorMult}
   B=MaterialFunctionCall() {materialFunction=/Game/Packages/Fortress_SharedMaterials/Base_Material_Functions/SkylightColorAndIntensity.SkyLightColorAndIntensity}.Color
  )
  n81 = CameraPositionWS()
  n82 = MaterialFunctionCall() {materialFunction=&WorldPositionBehindTranslucency}
  n83 = DotProduct(A=ReflectionVectorWS(CustomWorldNormal=@n14), B=Normalize(VectorInput=Subtract(A=@n81, B=@n82.Result)))
  n84 = Multiply(
   A=Power(Base=Saturate(None=@n83), Exp=ScalarParameter() {defaultValue=32, parameterName=SpecularPower})
   B=VectorParameter() {parameterName=SpecularColor, defaultValue=<rgba 1 1 1 0>}.RGB
  )
  n85 = ScalarParameter() {defaultValue=1, parameterName=SpecularStrength}
  n86 = Multiply(A=Add(A=Multiply(A=@n79, B=@n80), B=Multiply(A=@n84, B=@n85)), B=@n18)
  n87 = Multiply(A=@n86, B=ScalarParameter() {defaultValue=0.1, parameterName=Diffuse_Shaded})
  n88 = NamedRerouteUsage()
  Reroute(None=Multiply(A=LinearInterpolate(A=@n87, B=@n86, Alpha=@n88), B=ScalarParameter() {defaultValue=0.175, parameterName=EmissiveBoost}))
  Round(None=@n30)
  ScalarParameter() {defaultValue=200, parameterName=DEBUG_HeightDivisor}
  ScalarParameter() {defaultValue=20, parameterName=debug}
  SkyLightEnvMapSample()
  n89 = WorldPosition()
  n90 = Transform(None=Constant3Vector() {constant=<rgba 0 0 1 0>}) {transformSourceType=TRANSFORMSOURCE_Local}
  n91 = Transform(None=Constant3Vector() {constant=<rgba 0 1 0 0>}) {transformSourceType=TRANSFORMSOURCE_Local}
  n92 = Transform(None=Constant3Vector() {constant=<rgba 1 0 0 0>}) {transformSourceType=TRANSFORMSOURCE_Local}
  n93 = Normalize(VectorInput=Subtract(A=CameraPositionWS(), B=ObjectPositionWS()))
  n94 = ScalarParameter() {defaultValue=50, parameterName=CameraOffset}
  StaticSwitchParameter(
   True=Constant()
   False=Add(
    A=MaterialFunctionCall!(WorldPosition=@n89.XYZ, PivotLocation, WS Normals, Custom Object Basis 1=@n90, Custom Object Basis 2=@n91, Custom Object Basis 3=@n92) {materialFunction=/Engine/Functions/Engine_MaterialFunctions02/WorldPositionOffset/AlignMeshToTheCamera}.World Position Offset
    B=Multiply(A=@n93, B=@n94)
   )
  ) {group=Niagara, parameterName=Use Niagara Mesh Camera Facing}
 ]
}
